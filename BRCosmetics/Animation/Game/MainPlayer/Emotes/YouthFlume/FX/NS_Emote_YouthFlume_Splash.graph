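# Niagara system for the YouthFlume emote that produces a short water splash from three emitters.
# Edge_Spray_Mesh spawns curve-scaled spray meshes, while Splash_Drops and Splash_Foam burst gravity-driven sprites from offset, stretched cylinders.
# Dynamic material parameters animate each layer over particle life, and spray color switches on a performance-mode platform check.
# Front-end modifiers drop the droplets and continuous spray spawning, leaving the spray and foam bursts.
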NiagaraSystem /BRCosmetics/Animation/Game/MainPlayer/Emotes/YouthFlume/FX/NS_Emote_YouthFlume_Splash
dependencies {
 usedDataInterfaces [&NiagaraDataInterfaceCurve, &NiagaraDataInterfacePlatformSet, &NiagaraDataInterfaceVectorCurve]
 usedDynamicInputs []
 usedModules [&SystemState]
}
emitters {
 Edge_Spray_Mesh {
  emitterSpawnScript [/SetVariables(Emitter.OverallEmissiveAdd:NiagaraFloat=&FN_FrontEndFloatModifier(Gameplay=0, Front End=0.5)) {isSetParameters=true}]
  emitterUpdateScript [
   &EmitterState(Life Cycle Mode=::NewEnumerator1, Loop Behavior=::NewEnumerator1, Loop Duration=0.2, UseLoopDelay=false)
   &SpawnRate(
    SpawnRate=&FloatFromCurve(
     FloatCurve=DataInterface() {propertyValues {Curve {keys [<at 0 1>,<at 1 0>]}}}
     CurveIndex=$Emitter.NormalizedLoopAge
     Scale Curve=&Multiply_Float(A=50, B=&FN_FrontEndFloatModifier(Gameplay=1, Front End=0))
    )
    Use Spawn Probability=false
   )
   &SpawnBurst_Instantaneous(Spawn Count=30, Use Spawn Probability=false, UseLoopCountLimit=false)
  ]
  particleSpawnScript [
   &FN_SetSpawnGroup(Number of Spawn Groups=2)
   /SetVariables(Particles.FrontToBack:NiagaraFloat=&RandomRangeFloat(Minimum=0, Maximum=1)) {isSetParameters=true}
   &InitializeParticle(
    Color=&MakeCustomLinearColorFromBool(
     Bool=&IsPlatformSetActive(New Platform Set=DataInterface() {propertyValues {Platforms {
         qualityLevelMask=-31
         cVarConditions [{
           bUseMaxFloat=false
           bUseMaxInt=false
           bUseMinFloat=true
           bUseMinInt=true
           cVarName=athena.isperfmode
           failResponse=Disable
           maxFloat=1
           maxInt=1
           minFloat=1
           minInt=1
           passResponse=None
           value=true
          }]
        }}})
     True Color=<rgba 0.4282 1 0.9351 1>
     False Color=<rgba 1 1 1 0.6>
    )
    Lifetime Max=0.5
    Lifetime Mode=::NewEnumerator1
    Mesh Scale Max=<1.2 2.5 2>
    Mesh Scale Min=<0.6 1.5 1>
    Mesh Scale Mode=::NewEnumerator2
    UsePositionOffset=false
    Write Material Random=true
   ) {module=InitializeParticle001}
   &ShapeLocation(
    Cylinder Distribution=::NewEnumerator2
    Cylinder Height=10
    Cylinder Radius=40
    MaskSpawnGroup=false
    Non Uniform Scale=<0.6 1.8 1>
    Offset=&FN_FrontEndVector3Modifier(Gameplay=<0 -30 0>, Front End=<0 -30 0>)
    Offset Mode=::NewEnumerator0
    Rotation Axis=<0 0 1>
    Rotation Mode=::NewEnumerator1
    Shape Primitive=::NewEnumerator1
    U Position=&RandomRangeFloat(Minimum=0, Maximum=1)
    UseOwnerScale=true
    V Position=0
   )
   &AddVelocity(
    Module.Use Falloff Distance=false
    Origin Offset=&FN_FrontEndVector3Modifier(Gameplay=<0 -20 -30>, Front End=<20 40 -30>)
    Override Seed=false
    Velocity Mode=::NewEnumerator1
    Velocity Speed=&RandomRangeFloat(Minimum=1, Maximum=1)
   )
   /SetVariables(Particles.Scale=&Multiply_Vector(
     A=&VectorFromCurve(
      VectorCurve=DataInterface() {propertyValues {
        XCurve {keys [{arriveTangent=1.7846, interpMode=RCIM_Cubic, leaveTangent=1.7846, tangentMode=RCTM_User, time=0, value=0.4}, <at 1 1 RCIM_Cubic>]}
        YCurve {keys [{arriveTangent=1.7846, interpMode=RCIM_Cubic, leaveTangent=1.7846, tangentMode=RCTM_User, time=0, value=0.4}, <at 1 1 RCIM_Cubic>]}
        ZCurve {keys [{arriveTangent=1.7846, interpMode=RCIM_Cubic, leaveTangent=1.7846, tangentMode=RCTM_User, time=0, value=0.4}, <at 1 1 RCIM_Cubic>]}
       }}
      CurveIndex=$Particles.FrontToBack
      Scale Curve=&Multiply_VectorByFloat(Vector=$Particles.Scale, Float=&Lerp_Float(A=0.8, B=1, Alpha=1))
     )
     B=&Lerp_Vector(A=<0.6 0.6 0.6>, B=<0.6 0.6 0.6>, Alpha=$Particles.LeanLerp)
    )) {isSetParameters=true}
  ]
  particleUpdateScript [
   &ParticleState()
   /SetVariables(Particles.Scale=&Multiply_Vector(
     A=$Particles.Initial.Scale
     B=&VectorFromCurve(VectorCurve=DataInterface() {propertyValues {
        XCurve {keys [
          {arriveTangent=4.6994, interpMode=RCIM_Cubic, leaveTangent=4.6994, tangentMode=RCTM_User, time=0, value=0.0795}
          {arriveTangent=0.4589, interpMode=RCIM_Cubic, leaveTangent=0.4589, tangentMode=RCTM_User, time=0.1447, value=0.7609}
          <at 1 1>
         ]}
        YCurve {keys [{arriveTangent=2.5596, interpMode=RCIM_Cubic, leaveTangent=2.5596, tangentMode=RCTM_User, time=0, value=0.0767}, <at 1 1>]}
        ZCurve {keys [
          <at 0 0 RCIM_Cubic>
          {arriveTangent=-2.3619, interpMode=RCIM_Cubic, leaveTangent=-2.3619, tangentMode=RCTM_User, time=0.336, value=0.9567}
          <at 1 0>
         ]}
       }})
    )) {isSetParameters=true}
   &SolveForcesAndVelocity(Clamp Velocity=false, Limit Acceleration=false)
   &DynamicMaterialParameters(
    Index 0 Param 1=&FloatFromCurve(FloatCurve=DataInterface() {propertyValues {Curve {keys [
         {arriveTangent=1.2629, interpMode=RCIM_Cubic, leaveTangent=1.2629, tangentMode=RCTM_User, time=-0.0045, value=-1.9799}
         {arriveTangent=4.4591, interpMode=RCIM_Cubic, leaveTangent=4.4591, tangentMode=RCTM_User, time=1, value=0}
        ]}}})
    Index 0 Param 2=&FloatFromCurve(FloatCurve=DataInterface() {propertyValues {Curve {keys [<at 0.0102 4>,<at 0.9749 0>]}}})
    Index 0 Param 3=&MakeFloatFromVector(Vector=$Particles.Scale, Channel=::NewEnumerator0)
    Index 0 Param 4=1
    Index 1 Param 1=&FloatFromCurve(FloatCurve=DataInterface() {propertyValues {Curve {keys [<at 0.7 0 RCIM_Cubic>, {arriveTangent=6.4641, interpMode=RCIM_Cubic, leaveTangent=6.4641, tangentMode=RCTM_User, time=1, value=1}]}}})
    Module.Index0Float0Write=true
    Module.Index0Float1Write=true
    Module.Index0Float2Write=true
    Module.Index0Float3Write=true
    Module.Index1Float0Write=true
    Module.Index1Float1Write=true
    Module.Index1Float2Write=true
    Module.Index1Float3Write=true
    Write Parameter Index 1=::NewEnumerator1
   )
  ]
  rendererClasses [&NiagaraMeshRendererProperties]
  renderers [{rendererClass=&NiagaraMeshRendererProperties, rendererIndex=0}]
 }
 Splash_Drops {
  emitterUpdateScript [
   &EmitterState(Inactive Response=::NewEnumerator1, Life Cycle Mode=::NewEnumerator1, Loop Behavior=::NewEnumerator1, Loop Delay=0.2, UseLoopDelay=true)
   &SpawnBurst_Instantaneous(
    Spawn Count=&Multiply_Int(
     A=100
     B=&MakeIntFromFloat(Float=&FN_FrontEndFloatModifier(Gameplay=1, Front End=0))
    )
    Use Spawn Probability=false
    UseLoopCountLimit=false
   )
  ]
  particleSpawnScript [
   &FN_SetSpawnGroup(Number of Spawn Groups=2)
   /SetVariables(Particles.FrontToBack:NiagaraFloat=&RandomRangeFloat(Minimum=0, Maximum=1)) {isSetParameters=true}
   &InitializeParticle(
    Lifetime Max=0.4
    Lifetime Min=0.15
    Lifetime Mode=::NewEnumerator1
    Sprite Size Max=<5 10>
    Sprite Size Min=<3 6>
    Sprite Size Mode=::NewEnumerator2
    UsePositionOffset=false
    Write Material Random=true
   ) {module=InitializeParticle001}
   &ShapeLocation(
    Cylinder Distribution=::NewEnumerator2
    Cylinder Height=10
    Cylinder Radius=50
    MaskSpawnGroup=false
    Non Uniform Scale=<0.5 1.5 1>
    Offset=<0 -30 20>
    Offset Mode=::NewEnumerator0
    Rotation Axis=<0 0 1>
    Rotation Mode=::NewEnumerator1
    Shape Primitive=::NewEnumerator1
    U Position=&RandomRangeFloat(Minimum=0, Maximum=1)
    UseOwnerScale=true
    V Position=0
   )
   &AddVelocity(
    Module.Use Falloff Distance=false
    Origin Offset=<0 -80 -50>
    Override Seed=false
    Velocity Mode=::NewEnumerator1
    Velocity Speed=&RandomRangeFloat(Minimum=300, Maximum=500)
   ) {module=AddVelocity001}
   /Niagara/Modules/Update/SubUV/SubUVAnimation(SubUV Animation Mode=::NewEnumerator1)
  ]
  particleUpdateScript [
   &ParticleState()
   &SpriteRotationRate(Rotation Rate=&RandomRangeFloat(Minimum=-180, Maximum=180))
   &Drag(Drag=&RandomRangeFloat(Minimum=0, Maximum=0.5), Use Linear Drag=true, Use Rotational Drag=false)
   &GravityForce()
   &SolveForcesAndVelocity(Clamp Velocity=false, Limit Acceleration=false)
   &ScaleSpriteSize(Scale Factor=&Vector2DFromFloat(Value=&FloatFromCurve(
      FloatCurve=DataInterface() {propertyValues {Curve {keys [
          {interpMode=RCIM_Cubic, tangentMode=RCTM_User, time=0.8, value=1}
          {arriveTangent=-3.3474, interpMode=RCIM_Cubic, leaveTangent=-3.3474, tangentMode=RCTM_User, time=1, value=0}
         ]}}}
      Scale Curve=0.65
     )))
   &DynamicMaterialParameters(
    Index 0 Param 1=&RandomRangeFloat(Minimum=0, Maximum=1)
    Index 0 Param 3=&FloatFromCurve(FloatCurve=DataInterface() {propertyValues {Curve {keys [{interpMode=RCIM_Cubic, tangentMode=RCTM_User, time=0, value=0}, {interpMode=RCIM_Cubic, tangentMode=RCTM_User, time=1, value=1}]}}})
    Module.Index0Float0Write=true
    Module.Index0Float1Write=true
    Module.Index0Float2Write=true
    Module.Index0Float3Write=true
   )
  ]
  rendererClasses [&NiagaraSpriteRendererProperties]
  renderers [{rendererClass=&NiagaraSpriteRendererProperties}]
 }
 Splash_Foam {
  emitterUpdateScript [
   &EmitterState(
    Inactive Response=::NewEnumerator1
    Life Cycle Mode=::NewEnumerator1
    Loop Behavior=::NewEnumerator1
    Loop Delay=0.1
    Loop Duration=2
    UseLoopDelay=true
   )
   &SpawnBurst_Instantaneous(Spawn Count=30, Use Spawn Probability=false, UseLoopCountLimit=false)
  ]
  particleSpawnScript [
   &FN_SetSpawnGroup(Number of Spawn Groups=2)
   /SetVariables(Particles.FrontToBack:NiagaraFloat=&RandomRangeFloat(Minimum=0, Maximum=1)) {isSetParameters=true}
   &InitializeParticle(
    Lifetime Max=0.5
    Lifetime Min=0.2
    Lifetime Mode=::NewEnumerator1
    Mesh Scale Mode=::NewEnumerator3
    Sprite Rotation Mode=::NewEnumerator2
    Sprite Size Mode=::NewEnumerator1
    Uniform Sprite Size Max=121
    Uniform Sprite Size Min=50
    UsePositionOffset=false
    Write Material Random=true
   ) {module=InitializeParticle001}
   &ShapeLocation(
    Cylinder Distribution=::NewEnumerator2
    Cylinder Height=10
    Cylinder Radius=50
    MaskSpawnGroup=false
    Non Uniform Scale=<0.5 1.5 1>
    Offset=<0 -30 20>
    Offset Mode=::NewEnumerator0
    Rotation Axis=<0 0 1>
    Rotation Mode=::NewEnumerator1
    Shape Primitive=::NewEnumerator1
    U Position=&RandomRangeFloat(Minimum=0, Maximum=1)
    UseOwnerScale=true
    V Position=0
   )
   &AddVelocity(
    Module.Use Falloff Distance=false
    Origin Offset=<0 -80 -50>
    Override Seed=false
    Velocity Mode=::NewEnumerator1
    Velocity Speed=&RandomRangeFloat(Minimum=400, Maximum=100)
   ) {module=AddVelocity001}
   &SubUVAnimation(SubUV Animation Mode=::NewEnumerator1)
  ]
  particleUpdateScript [
   &ParticleState()
   &SpriteRotationRate(Rotation Rate=&RandomRangeFloat(Minimum=-90, Maximum=90))
   &Drag(Drag=0.5, Use Linear Drag=true, Use Rotational Drag=false)
   &GravityForce(Gravity=<0 0 -900>)
   &SolveForcesAndVelocity(Clamp Velocity=false, Limit Acceleration=false)
   &ScaleSpriteSize(Scale Factor=&Vector2DFromFloat(Value=&FloatFromCurve(
      FloatCurve=DataInterface() {propertyValues {Curve {keys [
          {arriveTangent=1.6115, interpMode=RCIM_Cubic, leaveTangent=1.6115, tangentMode=RCTM_User, time=0, value=0.33}
          {interpMode=RCIM_Cubic, tangentMode=RCTM_User, time=1, value=1}
         ]}}}
      Scale Curve=0.65
     )))
   &DynamicMaterialParameters(
    Index 0 Param 1=&RandomRangeFloat(Minimum=0, Maximum=1)
    Index 0 Param 3=&FloatFromCurve(FloatCurve=DataInterface() {propertyValues {Curve {keys [{interpMode=RCIM_Cubic, tangentMode=RCTM_User, time=0, value=0}, {interpMode=RCIM_Cubic, tangentMode=RCTM_User, time=1, value=1}]}}})
    Index 0 Param 4=&FloatFromCurve(FloatCurve=DataInterface() {propertyValues {Curve {keys [<at 0 0 RCIM_Cubic>, {arriveTangent=1.5276, interpMode=RCIM_Cubic, leaveTangent=1.5276, tangentMode=RCTM_User, time=1, value=1}]}}})
    Module.Index0Float0Write=true
    Module.Index0Float1Write=true
    Module.Index0Float2Write=true
    Module.Index0Float3Write=true
   )
  ]
  rendererClasses [&NiagaraSpriteRendererProperties]
  renderers [{rendererClass=&NiagaraSpriteRendererProperties}]
 }
}
summary {
 systemName=NS_Emote_YouthFlume_Splash
 emitters [
  {emitterName=Edge_Spray_Mesh, rendererClasses [&NiagaraMeshRendererProperties]}
  {emitterName=Splash_Foam, rendererClasses [&NiagaraSpriteRendererProperties]}
  {emitterName=Splash_Drops, rendererClasses [&NiagaraSpriteRendererProperties]}
 ]
 userVariables [
  User.bIsFrontEnd {type=NiagaraBool}
  User.bIsFrontEndPreview {type=NiagaraBool}
 ]
}
