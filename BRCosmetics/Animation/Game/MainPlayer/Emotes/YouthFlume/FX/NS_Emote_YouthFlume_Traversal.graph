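# Niagara system for the YouthFlume traversal emote, producing water effects around the moving character.
# Six socket-attached emitters layer mesh wakes, paired edge sprays, sprite splashes and drops, a ribbon water trail, and mesh-rendered "Ribbons".
# BlendSpace modules mix world and local simulation, often by owner speed, while User.ForwardDot raises the drop spawn rate.
# Colors and emissive adapt to time of day, performance mode or front-end display in selected emitters.
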NiagaraSystem /BRCosmetics/Animation/Game/MainPlayer/Emotes/YouthFlume/FX/NS_Emote_YouthFlume_Traversal
dependencies {
 usedDataInterfaces [
  &NiagaraDataInterfacePlatformSet
  &NiagaraDataInterfaceVectorCurve
  &NiagaraDataInterfaceCurve
  &NiagaraDataInterfaceArrayColor
  &NiagaraDataInterfaceArrayFloat
  &NiagaraDataInterfaceSocketReader
 ]
 usedDynamicInputs []
 usedModules [
  &self:SystemSpawnScript.SystemScriptSource.SystemScriptGraph.NiagaraNodeAssignment_0.SetVariables_#1
  &SystemState
  &BlendSpaceHelper
 ]
}
emitters {
 Edge_Spray_Mesh {
  emitterSpawnScript [/SetVariables(Emitter.OverallEmissiveAdd:NiagaraFloat=&FN_FrontEndFloatModifier(Gameplay=0, Front End=0.5)) {isSetParameters=true}]
  emitterUpdateScript [
   &EmitterState(Inactive Response=::NewEnumerator1, Life Cycle Mode=::NewEnumerator1, Loop Behavior=::NewEnumerator1, Loop Duration=3600, UseLoopDelay=false)
   &SpawnRate(
    SpawnGroup:NiagaraInt32=&RandomRangeInt(Minimum=0, Maximum=1)
    SpawnRate=60
    Use Spawn Probability=false
   )
  ]
  particleSpawnScript [
   &FN_SetSpawnGroup(Number of Spawn Groups=2)
   /SetVariables(Particles.FrontToBack:NiagaraFloat=&RandomRangeFloat(Minimum=0, Maximum=1)) {isSetParameters=true}
   &InitializeParticle(
    Color=&MakeCustomLinearColorFromBool(
     Bool=&IsPlatformSetActive(New Platform Set=DataInterface() {propertyValues {Platforms {
         qualityLevelMask=-31
         cVarConditions [{
           bUseMaxFloat=false
           bUseMaxInt=false
           bUseMinFloat=true
           bUseMinInt=true
           cVarName=athena.isperfmode
           failResponse=Disable
           maxFloat=1
           maxInt=1
           minFloat=1
           minInt=1
           passResponse=None
           value=true
          }]
        }}})
     True Color=<rgba 0.4282 1 0.9351 1>
     False Color=<rgba 1 1 1 1>
    )
    Lifetime Max=1
    Lifetime Min=0.75
    Lifetime Mode=::NewEnumerator1
    Mesh Scale=&Multiply_Vector(
     A=&Multiply_VectorByFloat(Vector=&RandomRangeVector(Minimum=<0.75 1.2 0.75>, Maximum=<1 2 1.1>), Float=0.5)
     B=&VectorFromCurve(
      VectorCurve=DataInterface() {propertyValues {XCurve {keys [<at 0 0.5291>,<at 1 1>]}, YCurve {keys [<at 0 1>,<at 1 1>]}, ZCurve {keys [<at 0 1>,<at 1 1>]}}}
      CurveIndex=$Particles.FrontToBack
     )
    )
    Mesh Scale Mode=::NewEnumerator4
    Sprite Rotation Mode=::NewEnumerator2
    UsePositionOffset=false
    Write Material Random=true
   ) {module=InitializeParticle001}
   &SocketLocation(
    SamplePosOffset=true
    Sampled Position Offset=&VectorFromCurve(
     VectorCurve=DataInterface() {propertyValues {
       XCurve {keys [
         <at 0 -71.8979>
         <at 0.4701 -113.1104>
         <at 0.7715 -108.9455>
         {arriveTangent=-382.5781, interpMode=RCIM_Cubic, leaveTangent=-382.5781, tangentMode=RCTM_User, time=1, value=-86.4235}
        ]}
       YCurve {keys [
         <at 0.33 -11.917 RCIM_Cubic>
         {arriveTangent=-92.7914, interpMode=RCIM_Cubic, leaveTangent=-92.7914, tangentMode=RCTM_User, time=1, value=-40}
        ]}
       ZCurve {keys [<at 0 1>,<at 1 1>]}
      }}
     CurveIndex=$Particles.FrontToBack
     Scale Curve=&Lerp_Vector(A=<-0.15 2 1>, B=<0.15 2 1>, Alpha=&MakeFloatFromInt(Int=$Particles.SpawnGroup))
    )
    Socket Reader=$System.Socket Reader
    Socket Sampling Mode=::NewEnumerator2
    VelocitySpeedLimit=false
   )
   &InitialMeshOrientation(
    Orientation Vector=&Multiply_Vector(
     A=&Add_Vector(A=&Lerp_Vector(A=<0 0 0>, B=<-1 0 0>, Alpha=$Particles.FrontToBack), B=<1 -1 0>)
     B=&Multiply_Vector(
      A=&Lerp_Vector(A=<1 1 1>, B=<1 -1 1>, Alpha=&MakeFloatFromInt(Int=$Particles.SpawnGroup))
      B=&Lerp_Vector(A=<1 1 1>, B=<1 0.75 1>, Alpha=1)
     )
    )
    Rotation=&Multiply_Vector(
     A=&MergeHelper_VectorSwitch(0=<0 0 -0.2521>, 1=<0 0 0.2597>, 2=?, 3=?, Index=&SpawnGroup(), Num=::NewEnumerator0)
     B=&VectorFromCurve(
      VectorCurve=DataInterface() {propertyValues {XCurve {keys [<at 0 1>,<at 1 1>]}, YCurve {keys [<at 0 1>,<at 1 1>]}, ZCurve {keys [<at 0 0.6757>,<at 0.7459 0.761>,<at 1 0.849>]}}}
      CurveIndex=$Particles.FrontToBack
     )
    )
    Rotation Coordinate Space=::NewEnumerator0
    Use Orientation Vector=true
    Use Rotation Vector=true
   )
   /SetVariables(Particles.Scale=&Multiply_Vector(
     A=&VectorFromCurve(
      VectorCurve=DataInterface() {propertyValues {
        XCurve {keys [{arriveTangent=1.7846, interpMode=RCIM_Cubic, leaveTangent=1.7846, tangentMode=RCTM_User, time=0, value=0.4}, <at 1 1 RCIM_Cubic>]}
        YCurve {keys [{arriveTangent=1.7846, interpMode=RCIM_Cubic, leaveTangent=1.7846, tangentMode=RCTM_User, time=0, value=0.4}, <at 1 1 RCIM_Cubic>]}
        ZCurve {keys [{arriveTangent=1.7846, interpMode=RCIM_Cubic, leaveTangent=1.7846, tangentMode=RCTM_User, time=0, value=0.4}, <at 1 1 RCIM_Cubic>]}
       }}
      CurveIndex=$Particles.FrontToBack
      Scale Curve=&Multiply_VectorByFloat(Vector=$Particles.Scale, Float=&Lerp_Float(A=0.8, B=1.4, Alpha=0.5))
     )
     B=&Lerp_Vector(A=<0.6 0.6 0.6>, B=<0.8 0.8 0.8>, Alpha=$Particles.LeanLerp)
    )) {isSetParameters=true}
  ]
  particleUpdateScript [
   &ParticleState()
   /SetVariables(Particles.Scale=&Multiply_Vector(
     A=$Particles.Initial.Scale
     B=&VectorFromCurve(VectorCurve=DataInterface() {propertyValues {
        XCurve {keys [
          {arriveTangent=4.6994, interpMode=RCIM_Cubic, leaveTangent=4.6994, tangentMode=RCTM_User, time=0, value=0.0795}
          {arriveTangent=0.4589, interpMode=RCIM_Cubic, leaveTangent=0.4589, tangentMode=RCTM_User, time=0.1447, value=0.7609}
          <at 1 1>
         ]}
        YCurve {keys [{arriveTangent=2.5596, interpMode=RCIM_Cubic, leaveTangent=2.5596, tangentMode=RCTM_User, time=0, value=0.0767}, <at 1 1>]}
        ZCurve {keys [
          <at 0 0 RCIM_Cubic>
          {arriveTangent=-2.3619, interpMode=RCIM_Cubic, leaveTangent=-2.3619, tangentMode=RCTM_User, time=0.336, value=1.5}
          <at 0.9988 0>
         ]}
       }})
    )) {isSetParameters=true}
   &SolveForcesAndVelocity(Clamp Velocity=false, Limit Acceleration=false)
   &DynamicMaterialParameters(
    Index 0 Param 1=&FloatFromCurve(FloatCurve=DataInterface() {propertyValues {Curve {keys [
         {arriveTangent=1.2629, interpMode=RCIM_Cubic, leaveTangent=1.2629, tangentMode=RCTM_User, time=-0.0045, value=-1.9799}
         {arriveTangent=4.4591, interpMode=RCIM_Cubic, leaveTangent=4.4591, tangentMode=RCTM_User, time=1, value=0}
        ]}}})
    Index 0 Param 2=&FloatFromCurve(FloatCurve=DataInterface() {propertyValues {Curve {keys [<at 0.0102 4>,<at 0.9749 0>]}}})
    Index 0 Param 3=&MakeFloatFromVector(Vector=$Particles.Scale, Channel=::NewEnumerator0)
    Index 0 Param 4=1
    Index 1 Param 1=&FloatFromCurve(FloatCurve=DataInterface() {propertyValues {Curve {keys [<at 0.7 0 RCIM_Cubic>, {arriveTangent=6.4641, interpMode=RCIM_Cubic, leaveTangent=6.4641, tangentMode=RCTM_User, time=1, value=1}]}}})
    Module.Index0Float0Write=true
    Module.Index0Float1Write=true
    Module.Index0Float2Write=true
    Module.Index0Float3Write=true
    Module.Index1Float0Write=true
    Module.Index1Float1Write=true
    Module.Index1Float2Write=true
    Module.Index1Float3Write=true
    Write Parameter Index 1=::NewEnumerator1
   )
   &ScaleColor(
    Scale Alpha=&FloatFromCurve(FloatCurve=DataInterface() {propertyValues {Curve {keys [{interpMode=RCIM_Cubic, tangentMode=RCTM_User, time=0, value=0}, {interpMode=RCIM_Cubic, tangentMode=RCTM_User, time=0.16, value=1}]}}})
    ScaleA=true
    ScaleRGB=true
   )
   &BlendSpace(Rotation Contribution=0.9)
  ]
  rendererClasses [&NiagaraMeshRendererProperties]
  renderers [{rendererClass=&NiagaraMeshRendererProperties, rendererIndex=0}]
 }
 Ribbons {
  emitterUpdateScript [
   &EmitterState(
    Inactive Response=::NewEnumerator1
    Life Cycle Mode=::NewEnumerator1
    Loop Behavior=::NewEnumerator1
    Loop Delay=1.5
    Loop Duration Mode=::NewEnumerator1
    UseLoopDelay=false
   )
   &SpawnBurst_Instantaneous(Spawn Count=2, Spawn Time=0.5, Use Spawn Probability=false, UseLoopCountLimit=false)
   &SpawnBurst_Instantaneous(Spawn Count=2, Spawn Group=1, Spawn Time=0.5, Use Spawn Probability=false, UseLoopCountLimit=false) {module=SpawnBurst_Instantaneous001}
  ]
  particleSpawnScript [
   /SetVariables(Particles.InitialSpawnGroup:NiagaraInt32=&SpawnGroup()) {isSetParameters=true}
   &FN_SetSpawnGroup()
   /SetVariables(Particles.VisibilityTag=&MergeHelper_IntSwitch(0=1, 1=0, 2=0, 3=1, 4=?, 5=?, Index=$Particles.SpawnGroup, Num=::NewEnumerator2)) {isSetParameters=true}
   &PartitionParticles()
   &InitializeParticle(
    Color=<rgba 1 1 1 0.1>
    Lifetime=0.2
    Mesh Scale=&Multiply_VectorByFloat(
     Vector=&MakeVector(
      X=&SelectFloatFromArray(Direct Array Index=$Particles.InitialSpawnGroup, Float Selection Array=DataInterface() {propertyValues {FloatData [1,-1]}})
      Y=1
      Z=1
     )
     Float=0.5
    )
    Mesh Scale Mode=::NewEnumerator4
    UsePositionOffset=false
    Write Material Random=true
   ) {module=InitializeParticle001}
   &SocketLocation(
    SamplePosOffset=true
    Sampled Position Offset=&MergeHelper_VectorSwitch(0=<2 -10 30>, 1=<-2 -10 30>, 2=?, 3=?, Index=$Particles.PartitionParticles.Partition, Num=::NewEnumerator0)
    Socket Reader=$System.Socket Reader
    Socket Sampling Mode=::NewEnumerator3
    VelocitySpeedLimit=false
   )
   &CameraOffset(Camera Offset Amount=5)
  ]
  particleUpdateScript [
   &ParticleState(Kill Particles When Lifetime Has Elapsed=false)
   &SolveForcesAndVelocity(Clamp Velocity=false, Limit Acceleration=false)
   &InterpolateOverTime(
    Rate Of Change=10
    Target Vector Value:Vector3f=&TransformPosition(Position=$Particles.Position, Source Space=::Local, Destination Space=::World)
    Variable Type=::NewEnumerator2
   )
   &DynamicMaterialParameters(
    Index 0 Param 1=&RandomRangeFloat(Minimum=0, Maximum=1)
    Index 0 Param 3=&MakeFloatFromVector(Vector=<0 0 0>, Channel=::NewEnumerator0)
    Index 1 Param 1=&MakeFloatFromVector(Vector=$StackContext.InterpolateOverTime.InterpolateOverTime_V3_.Moving Average, Channel=::NewEnumerator0)
    Index 1 Param 2=&MakeFloatFromVector(Vector=$StackContext.InterpolateOverTime.InterpolateOverTime_V3_.Moving Average, Channel=::NewEnumerator1)
    Index 1 Param 3=&MakeFloatFromVector(Vector=$StackContext.InterpolateOverTime.InterpolateOverTime_V3_.Moving Average, Channel=::NewEnumerator2)
    Module.Index0Float0Write=true
    Module.Index0Float1Write=true
    Module.Index0Float2Write=true
    Module.Index0Float3Write=true
    Module.Index1Float0Write=true
    Module.Index1Float1Write=true
    Module.Index1Float2Write=true
    Module.Index1Float3Write=true
    Write Parameter Index 1=::NewEnumerator1
   ) {module=DynamicMaterialParameters001}
   &ScaleColor(
    Scale Alpha=&FloatFromCurve(FloatCurve=DataInterface() {propertyValues {Curve {keys [{interpMode=RCIM_Cubic, tangentMode=RCTM_User, time=0, value=0}, {interpMode=RCIM_Cubic, tangentMode=RCTM_User, time=1, value=1}]}}})
    Scale RGB=&MakeVectorFromLinearColorRGB(Color=&FNTimeOfDayColor(Night Color=<rgba 0.334 0.556 1 1>, Morning Color=<rgba 1 1 1 1>, Day Color=<rgba 1 1 1 1>, Evening Color=<rgba 1 1 1 1>))
    ScaleA=true
    ScaleRGB=true
   )
  ]
  rendererClasses [&NiagaraMeshRendererProperties]
  renderers [
   {rendererClass=&NiagaraMeshRendererProperties, rendererIndex=0}
   {rendererClass=&NiagaraMeshRendererProperties, rendererIndex=1}
  ]
 }
 Splash_Drops {
  emitterUpdateScript [
   &EmitterState(Inactive Response=::NewEnumerator1, Life Cycle Mode=::NewEnumerator1, Loop Behavior=::NewEnumerator1, Loop Duration=3600, UseLoopDelay=false)
   &SpawnRate(
    SpawnGroup:NiagaraInt32=&RandomRangeInt(Minimum=0, Maximum=1)
    SpawnRate=&Lerp_Float(A=120, B=200, Alpha=&ClampFloat(Float=$User.ForwardDot))
    Use Spawn Probability=false
   )
  ]
  particleSpawnScript [
   &FN_SetSpawnGroup(Number of Spawn Groups=2)
   /SetVariables(Particles.FrontToBack:NiagaraFloat=&RandomRangeFloat(Minimum=0, Maximum=1)) {isSetParameters=true}
   &InitializeParticle(
    Color Minimum=<rgba 0.8377 0.8662 1 1>
    Color Mode=::NewEnumerator2
    Lifetime Max=0.6
    Lifetime Min=0.3
    Lifetime Mode=::NewEnumerator1
    Sprite Rotation Mode=::NewEnumerator2
    Sprite Size Mode=::NewEnumerator1
    Uniform Sprite Size Max=4
    Uniform Sprite Size Min=3
    UsePositionOffset=false
    Write Material Random=true
   ) {module=InitializeParticle001}
   &SocketLocation(
    SamplePosOffset=true
    Sampled Position Offset=&VectorFromCurve(
     VectorCurve=DataInterface() {propertyValues {
       XCurve {keys [
         <at 0 -71.8979>
         <at 0.4701 -113.1104>
         <at 0.7715 -108.9455>
         {arriveTangent=-382.5781, interpMode=RCIM_Cubic, leaveTangent=-382.5781, tangentMode=RCTM_User, time=1, value=-86.4235}
        ]}
       YCurve {keys [
         <at 0.33 -11.917 RCIM_Cubic>
         {arriveTangent=-92.7914, interpMode=RCIM_Cubic, leaveTangent=-92.7914, tangentMode=RCTM_User, time=1, value=-35}
        ]}
       ZCurve {keys [<at 0 0>,<at 1 0>]}
      }}
     CurveIndex=$Particles.FrontToBack
     Scale Curve=&Lerp_Vector(A=<-0.2 3 1>, B=<0.2 3 1>, Alpha=&MakeFloatFromInt(Int=$Particles.SpawnGroup))
    )
    Socket Reader=$System.Socket Reader
    Socket Sampling Mode=::NewEnumerator2
    VelocitySpeedLimit=false
   )
   /Niagara/Modules/Spawn/Location/SphereLocation(MaskSpawnGroup=false, Module.Override Randomness=false, Sphere Radius=3, Surface Only=false)
   /Niagara/Modules/Update/SubUV/SubUVAnimation(SubUV Animation Mode=::NewEnumerator1)
   /SetVariables(Particles.Velocity=&Multiply_VectorByFloat(
     Vector=&ReturnMeshOrientationAxis(Mesh Orientation Axis=::YAxis, Mesh Orientation Quat=$Particles.MeshOrientation)
     Float=-100
    )) {isSetParameters=true}
   &AddVelocity(
    Scale Added Velocity=&Lerp_Vector(A=<1 1 1>, B=<1.1 1.1 1.25>, Alpha=1)
    Velocity=&RandomRangeVector(Minimum=<40 0 50>, Maximum=<-40 0 200>)
   )
   &AddVelocityFromPoint(Module.Use Falloff Distance=false, Origin Offset=<0 -20 -15>, Origin Offset Coordinate Space=::Simulation)
  ]
  particleUpdateScript [
   &ParticleState()
   &InheritVelocity(
    Module.Limit Speed=true
    Velocity Limit=10000
    Velocity Scale=&Multiply_VectorByFloat(Vector=<-0.03 -0.03 -0.1>, Float=0.1)
    Velocity Source=&TransformVector(Vector=$Engine.Owner.Velocity, Source Space=::World, Destination Space=::Local)
   )
   &SpriteRotationRate(Rotation Rate=&RandomRangeFloat(Minimum=-180, Maximum=180))
   &Drag(Drag=0.5, Use Linear Drag=true, Use Rotational Drag=false)
   &GravityForce(Gravity=<0 0 -900>)
   &SolveForcesAndVelocity(Clamp Velocity=false, Limit Acceleration=false)
   &ScaleSpriteSize(Scale Factor=&Vector2DFromFloat(Value=&FloatFromCurve(
      FloatCurve=DataInterface() {propertyValues {Curve {keys [
          {interpMode=RCIM_Cubic, tangentMode=RCTM_User, time=0.8, value=1}
          {arriveTangent=-3.3474, interpMode=RCIM_Cubic, leaveTangent=-3.3474, tangentMode=RCTM_User, time=1, value=0}
         ]}}}
      Scale Curve=0.65
     )))
   &ScaleSpriteSizeBySpeed(Max Scale Factor=<0.5 6>, Min Scale Factor=<0.5 1>)
   &DynamicMaterialParameters(
    Index 0 Param 1=&RandomRangeFloat(Minimum=0, Maximum=1)
    Index 0 Param 3=&FloatFromCurve(FloatCurve=DataInterface() {propertyValues {Curve {keys [{interpMode=RCIM_Cubic, tangentMode=RCTM_User, time=0, value=0}, {interpMode=RCIM_Cubic, tangentMode=RCTM_User, time=1, value=1}]}}})
    Module.Index0Float0Write=true
    Module.Index0Float1Write=true
    Module.Index0Float2Write=true
    Module.Index0Float3Write=true
   )
   &ScaleColor(
    Scale Alpha=&FloatFromCurve(FloatCurve=DataInterface() {propertyValues {Curve {keys [
         {interpMode=RCIM_Cubic, tangentMode=RCTM_User, time=0, value=0}
         {interpMode=RCIM_Cubic, tangentMode=RCTM_User, time=0.17, value=1}
         {interpMode=RCIM_Cubic, tangentMode=RCTM_User, time=0.6, value=1}
         {interpMode=RCIM_Cubic, tangentMode=RCTM_User, time=1, value=0}
        ]}}})
    ScaleA=true
    ScaleRGB=true
   )
   &BlendSpace("Position Blend (World <-> Local)":NiagaraFloat=&MakeCustomFloatFromBool(
     Bool=&SetBoolByFloatComparison(A=&VectorLength(Vector=$Engine.Owner.Velocity), B=350)
     True Float=1
     False Float=&FloatFromCurve(FloatCurve=DataInterface() {propertyValues {Curve {keys [<at 0.1106 1 RCIM_Cubic>,<at 0.5991 0.7 RCIM_Cubic>]}}})
     Translation Contribution:NiagaraFloat=1
     Rotation Contribution:NiagaraFloat=1
    ))
  ]
  rendererClasses [&NiagaraSpriteRendererProperties]
  renderers [{rendererClass=&NiagaraSpriteRendererProperties}]
 }
 Splashes {
  emitterUpdateScript [
   &EmitterState(Inactive Response=::NewEnumerator1, Life Cycle Mode=::NewEnumerator1, Loop Behavior=::NewEnumerator1, Loop Duration=3600, UseLoopDelay=false)
   &SpawnRate(SpawnGroup=2, SpawnRate=60, Use Spawn Probability=false) {module=SpawnRate001}
   &SpawnRate(SpawnRate=40, Use Spawn Probability=false)
  ]
  particleSpawnScript [
   &FN_SetSpawnGroup(Number of Spawn Groups=3)
   &InitializeParticle(
    Color=&SelectLinearColorFromArray(
     Array Sampling Mode=::NewEnumerator0
     Direct Array Index=?
     Color Selection Array=DataInterface() {propertyValues {ColorData [<rgba 40 40 40 0.1>, <rgba 2.998 3.855 5 0.6889>]}}
     Random Seed=0
    )
    Lifetime Max=0.4
    Lifetime Min=0.2
    Lifetime Mode=::NewEnumerator1
    Mesh Scale Mode=::NewEnumerator3
    Sprite Rotation Mode=::NewEnumerator2
    Sprite Size Mode=::NewEnumerator1
    Uniform Sprite Size Max=&MaskFloat(Mask Source=0, Mask Parameter=?, Mask=2, Masked Float=40, Passthrough Float=70)
    Uniform Sprite Size Min=50
    UsePositionOffset=false
    Write Material Random=true
   ) {module=InitializeParticle001}
   &SocketLocation(
    SamplePosOffset=true
    Sampled Position Offset=&MergeHelper_VectorSwitch(0=<15 -10 -5>, 1=<-15 -10 -5>, 2=<0 -10 7>, 3=?, Index=&SpawnGroup(), Num=::NewEnumerator1)
    Socket Reader=$System.Socket Reader
    Socket Sampling Mode=::NewEnumerator3
    VelocitySpeedLimit=false
   )
   &ShapeLocation(
    Box Size=<40 1 1>
    MaskSpawnGroup=true
    Override Seed=false
    Shape Origin=$Output.SocketLocation.SampledPosition
    Shape Primitive=::NewEnumerator2
    Spawn Group:NiagaraInt32=&SpawnGroup()
    Spawn Group Mask=2
    Surface Only=false
    UseOwnerScale=true
   )
   &SubUVAnimation(SubUV Animation Mode=::NewEnumerator1)
   &AddVelocity(Velocity=&MaskVector(
     Mask Source=0
     Mask Parameter=?
     Mask=2
     Masked Vector=<0 0 0>
     Passthrough Vector=&Add_Vector(
      A=&RandomRangeVector(Minimum=<0 -200 100>, Maximum=<0 -450 200>)
      B=&MaskVectorBySpawnGroup(
       Masked Vector=&RandomRangeVector(Minimum=<100 0 0>, Maximum=<150 0 0>)
       Passthrough Vector=&RandomRangeVector(Minimum=<-150 0 0>, Maximum=<-100 0 0>)
      )
     )
    ))
   &AddVelocityFromPoint(
    Module.Use Falloff Distance=false
    Origin Offset=&MergeHelper_VectorSwitch(0=<-4000 0 400>, 1=<4000 0 100>, 2=<0 0 0>, 3=?, Index=&SpawnGroup(), Num=::NewEnumerator1)
    Origin Offset Coordinate Space=::Simulation
    Velocity Strength=30
   )
   &CameraOffset(Camera Offset Amount=&MaskFloatBySpawnGroup(Masked Float=-5, Passthrough Float=0, Spawn Group Mask=2))
  ]
  particleUpdateScript [
   &ParticleState()
   &SpriteRotationRate(Rotation Rate=&RandomRangeFloat(Minimum=-90, Maximum=90))
   &Drag(Drag=0.5, Use Linear Drag=true, Use Rotational Drag=false)
   &GravityForce(Gravity=&MaskVector(Mask Source=0, Mask Parameter=?, Mask=2, Masked Vector=<0 0 0>, Passthrough Vector=<0 0 -600>))
   &SolveForcesAndVelocity(Clamp Velocity=false, Limit Acceleration=false)
   &ScaleSpriteSize(Scale Factor=&Vector2DFromFloat(Value=&FloatFromCurve(
      FloatCurve=DataInterface() {propertyValues {Curve {keys [
          {arriveTangent=1.6115, interpMode=RCIM_Cubic, leaveTangent=1.6115, tangentMode=RCTM_User, time=0, value=0.33}
          {interpMode=RCIM_Cubic, tangentMode=RCTM_User, time=1, value=1}
         ]}}}
      Scale Curve=0.5
     )))
   &DynamicMaterialParameters(
    Index 0 Param 1=&RandomRangeFloat(Minimum=0, Maximum=1)
    Index 0 Param 3=&FloatFromCurve(FloatCurve=DataInterface() {propertyValues {Curve {keys [{interpMode=RCIM_Cubic, tangentMode=RCTM_User, time=0, value=0}, {interpMode=RCIM_Cubic, tangentMode=RCTM_User, time=1, value=1}]}}})
    Index 0 Param 4=&FloatFromCurve(FloatCurve=DataInterface() {propertyValues {Curve {keys [<at 0 0 RCIM_Cubic>, {arriveTangent=1.5276, interpMode=RCIM_Cubic, leaveTangent=1.5276, tangentMode=RCTM_User, time=1, value=1}]}}})
    Module.Index0Float0Write=true
    Module.Index0Float1Write=true
    Module.Index0Float2Write=true
    Module.Index0Float3Write=true
   )
   &ScaleColor(
    Scale Alpha=&FloatFromCurve(FloatCurve=DataInterface() {propertyValues {Curve {keys [<at 0 0 RCIM_Cubic>,<at 0.3 1 RCIM_Cubic>,<at 1 0 RCIM_Cubic>]}}}, Scale Curve=2)
    ScaleA=true
    ScaleRGB=false
   )
   &BlendSpace("Position Blend (World <-> Local)":NiagaraFloat=&MaskFloat(
     Mask Source=0
     Mask Parameter=?
     Mask=&SpawnGroup()
     Masked Float=1
     Passthrough Float=&MakeCustomFloatFromBool(
      Bool=&SetBoolByFloatComparison(A=&VectorLength(Vector=$Engine.Owner.Velocity), B=350)
      True Float=1
      False Float=&FloatFromCurve(FloatCurve=DataInterface() {propertyValues {Curve {keys [<at 0.1106 1 RCIM_Cubic>,<at 0.5991 0.7 RCIM_Cubic>]}}})
     )
     Translation Contribution:NiagaraFloat=1
     Rotation Contribution:NiagaraFloat=1
    ))
  ]
  rendererClasses [&NiagaraSpriteRendererProperties]
  renderers [{rendererClass=&NiagaraSpriteRendererProperties}]
 }
 Wakes {
  emitterSpawnScript [/SetVariables(
    Emitter.OverallEmissiveAdd:NiagaraFloat=&FN_FrontEndFloatModifier(Gameplay=0, Front End=0.3)
    Emitter.ReflectionIntensity:NiagaraFloat=&FN_FrontEndFloatModifier(Gameplay=2, Front End=2)
   ) {isSetParameters=true}]
  emitterUpdateScript [
   &EmitterState(Delay First Loop Only=true, Inactive Response=::NewEnumerator2, Life Cycle Mode=::NewEnumerator1, Loop Duration=4000, UseLoopDelay=true)
   &SpawnRate(SpawnRate=8, Use Spawn Probability=false)
  ]
  particleSpawnScript [
   &InitializeParticle(
    Color=&MakeCustomLinearColorFromBool(
     Bool=$User.bIsFrontEndPreview
     True Color=<rgba 1 1 1 1>
     False Color=&MakeCustomLinearColorFromBool(
      Bool=&IsPlatformSetActive(New Platform Set=DataInterface() {propertyValues {Platforms {
          qualityLevelMask=-31
          cVarConditions [{
            bUseMaxFloat=false
            bUseMaxInt=false
            bUseMinFloat=true
            bUseMinInt=true
            cVarName=athena.isperfmode
            failResponse=Disable
            maxFloat=1
            maxInt=1
            minFloat=1
            minInt=1
            passResponse=None
            value=true
           }]
         }}})
      True Color=<rgba 0.4282 1 0.9351 1>
      False Color=<rgba 1 1 1 0.6>
     )
    )
    Lifetime=0.9
    Mesh Scale=&Multiply_VectorByFloat(Vector=&UniformRangedVector(Minimum=<1.2 1.3 0.1>, Maximum=<1.3 1 0.1>), Float=0.6)
    Mesh Scale Mode=::NewEnumerator4
    Position Mode=::NewEnumerator0
    Sprite Size Mode=::NewEnumerator3
    Write Material Random=true
   ) {module=InitializeParticle001}
   &SocketLocation(
    SamplePosOffset=true
    Sampled Position Offset=<0 -60 5>
    Sampled Position Offset Coordinate Space:ENiagaraSkelMeshTransforms=::NewEnumerator2
    Socket Reader=$System.Socket Reader
    Socket Sampling Mode=::NewEnumerator3
    VelocitySpeedLimit=false
   )
   &AddVelocity(Velocity=<0 0 5>)
   &InitialMeshOrientation(Mesh Orientation Mode=::NewEnumerator2, Orientation Vector=<0 1 0>, Use Rotation Vector=true)
  ]
  particleUpdateScript [
   &ParticleState()
   &ScaleMeshSize(Scale Factor=&VectorFromCurve(VectorCurve=DataInterface() {propertyValues {
       XCurve {keys [
         {arriveTangent=4.6994, interpMode=RCIM_Cubic, leaveTangent=4.6994, tangentMode=RCTM_User, time=0, value=0.2795}
         {arriveTangent=0.4589, interpMode=RCIM_Cubic, leaveTangent=0.4589, tangentMode=RCTM_User, time=0.15, value=0.7609}
         <at 1 1>
        ]}
       YCurve {keys [
         {arriveTangent=1.7936, interpMode=RCIM_Cubic, leaveTangent=1.7936, tangentMode=RCTM_User, time=0, value=0.2}
         <at 0.4 0.4>
         {arriveTangent=-0.1215, interpMode=RCIM_Cubic, leaveTangent=-0.1215, tangentMode=RCTM_User, time=1, value=1}
        ]}
       ZCurve {keys [<at 0 0.2 RCIM_Cubic>, {arriveTangent=-0.2002, interpMode=RCIM_Cubic, leaveTangent=-0.2002, time=0.33, value=1}, <at 0.9988 0>]}
      }}))
   &Drag(Drag=0, Rotational Drag=20, Use Linear Drag=true, Use Rotational Drag=true)
   &ScaleColor(
    Scale Alpha=&FloatFromCurve(
     FloatCurve=DataInterface() {propertyValues {Curve {keys [
         {arriveTangent=9.9926, interpMode=RCIM_Cubic, leaveTangent=9.9927, tangentMode=RCTM_User, time=0, value=0}
         {arriveTangent=-2.8339, interpMode=RCIM_Cubic, leaveTangent=-2.8339, tangentMode=RCTM_User, time=0.0216, value=1}
         <at 1.0011 0.002 RCIM_Cubic>
        ]}}}
     Scale Curve=3
    )
    ScaleA=true
    ScaleRGB=false
   )
   &SolveForcesAndVelocity(Clamp Velocity=false, Limit Acceleration=false)
   &DynamicMaterialParameters(
    Index 0 Param 1=&FloatFromCurve(FloatCurve=DataInterface() {propertyValues {Curve {keys [
         {arriveTangent=0.2615, interpMode=RCIM_Cubic, leaveTangent=0.2615, tangentMode=RCTM_User, time=0, value=-1.7984}
         {arriveTangent=3.2109, interpMode=RCIM_Cubic, leaveTangent=3.2109, tangentMode=RCTM_User, time=0.9985, value=0}
        ]}}})
    Index 0 Param 2=&Add_Float(
     A=&FloatFromCurve(FloatCurve=DataInterface() {propertyValues {Curve {keys [
          {arriveTangent=-9.4845, interpMode=RCIM_Cubic, leaveTangent=-9.4845, tangentMode=RCTM_User, time=0, value=4}
          {arriveTangent=0.2383, interpMode=RCIM_Cubic, leaveTangent=0.2383, tangentMode=RCTM_User, time=1, value=0}
         ]}}})
     B=0
    )
    Index 1 Param 1=&FloatFromCurve(FloatCurve=DataInterface() {propertyValues {Curve {keys [
         <at 0 0.1 RCIM_Cubic>
         {arriveTangent=0.9, interpMode=RCIM_Cubic, leaveTangent=0.9, time=0.5861, value=0.2}
         {arriveTangent=3.3294, interpMode=RCIM_Cubic, leaveTangent=3.3294, tangentMode=RCTM_User, time=1, value=1}
        ]}}})
    Module.Index0Float0Write=true
    Module.Index0Float1Write=true
    Module.Index0Float2Write=true
    Module.Index0Float3Write=true
    Module.Index1Float0Write=true
    Module.Index1Float1Write=true
    Module.Index1Float2Write=true
    Module.Index1Float3Write=true
    Write Parameter Index 1=::NewEnumerator1
   )
   &BlendSpace(
    "Mesh Orientation Blend (World <-> Local)":NiagaraFloat=&MakeCustomFloatFromBool(
     Bool=&SetBoolByFloatComparison(A=&VectorLength(Vector=$Engine.Owner.Velocity), B=350)
     True Float=1
     False Float=0.3
    )
    Mesh Orientation Mode=1
    "Position Blend (World <-> Local)":NiagaraFloat=&MakeCustomFloatFromBool(
     Bool=&SetBoolByFloatComparison(A=&VectorLength(Vector=$Engine.Owner.Velocity), B=350)
     True Float=1
     False Float=&FloatFromCurve(FloatCurve=DataInterface() {propertyValues {Curve {keys [<at 0 1>,<at 0.6 0.4>]}}})
     Translation Contribution:NiagaraFloat=1
     Rotation Contribution:NiagaraFloat=1
    )
   )
  ]
  rendererClasses [&NiagaraMeshRendererProperties]
  renderers [{rendererClass=&NiagaraMeshRendererProperties, rendererIndex=0}]
 }
 WaterTrail {
  emitterUpdateScript [
   &EmitterState(Life Cycle Mode=::NewEnumerator1, Loop Behavior=::NewEnumerator1, Loop Delay=0.389, Loop Duration Mode=::NewEnumerator1, UseLoopDelay=true)
   &SpawnRate(SpawnRate=50, Use Spawn Probability=false)
  ]
  particleSpawnScript [
   &InitializeParticle(
    Color=&FNTimeOfDayColor(
     Night Color=<rgba 0.2086 0.3957 1 0.828>
     Morning Color=<rgba 0.3895 0.6132 1 0.828>
     Day Color=<rgba 0.3895 0.6132 1 0.828>
     Evening Color=<rgba 0.3895 0.6132 1 0.828>
    )
    Lifetime=0.5
    Ribbon Facing Mode=::NewEnumerator1
    Ribbon Facing Vector:Vector3f=$Engine.Owner.SystemZAxis
    Ribbon Width=150
    Ribbon Width Mode=::NewEnumerator1
    UsePositionOffset=false
    Write Material Random=true
   )
   &SocketLocation(
    SamplePosOffset=true
    Sampled Position Offset=<0 0 5>
    Sampled Position Offset Coordinate Space:ENiagaraSkelMeshTransforms=::NewEnumerator2
    Socket Reader=$System.Socket Reader
    Socket Sampling Mode=::NewEnumerator3
    VelocitySpeedLimit=false
   )
   &AddVelocity(Velocity=&MakeVector(X=0, Y=-1000, Z=0))
  ]
  particleUpdateScript [
   &ParticleState()
   &ScaleColor(
    Scale Alpha=&FloatFromCurve(
     FloatCurve=DataInterface() {propertyValues {Curve {keys [
         {interpMode=RCIM_Cubic, tangentMode=RCTM_Break, time=0, value=0}
         {arriveTangent=1.6363, interpMode=RCIM_Cubic, leaveTangent=1.6363, time=0.15, value=1}
         {arriveTangent=-1.5385, interpMode=RCIM_Cubic, leaveTangent=-1.5385, time=0.4871, value=0.797}
         {interpMode=RCIM_Cubic, tangentMode=RCTM_Break, time=0.8, value=0}
        ]}}}
     Scale Curve=0.7
    )
    ScaleA=true
    ScaleRGB=true
   )
   &ScaleRibbonWidth(Ribbon Width Scale=&FloatFromCurve(FloatCurve=DataInterface() {propertyValues {Curve {keys [
         <at 0.0009 0.3 RCIM_Cubic>
         {interpMode=RCIM_Cubic, tangentMode=RCTM_Break, time=0.2, value=1}
         {arriveTangent=-1, interpMode=RCIM_Cubic, leaveTangent=-1, time=0.7536, value=0.9082}
         {arriveTangent=-2.4767, interpMode=RCIM_Cubic, tangentMode=RCTM_Break, time=1, value=0.2}
        ]}}}))
   &Drag(Drag=3, Use Linear Drag=true, Use Rotational Drag=true)
   &GravityForce(Gravity=<-500 0 0>)
   &SolveForcesAndVelocity(Clamp Velocity=false, Limit Acceleration=false)
   &BlendSpace(Rotation Contribution=0)
  ]
  rendererClasses [&NiagaraRibbonRendererProperties]
  renderers [{rendererClass=&NiagaraRibbonRendererProperties}]
 }
}
summary {
 systemName=NS_Emote_YouthFlume_Traversal
 emitters [
  {emitterName=Wakes, rendererClasses [&NiagaraMeshRendererProperties]}
  {emitterName=Splashes, rendererClasses [&NiagaraSpriteRendererProperties]}
  {emitterName=Ribbons, rendererClasses [&NiagaraMeshRendererProperties]}
  {emitterName=Edge_Spray_Mesh, rendererClasses [&NiagaraMeshRendererProperties]}
  {emitterName=Splash_Drops, rendererClasses [&NiagaraSpriteRendererProperties]}
  {emitterName=WaterTrail, rendererClasses [&NiagaraRibbonRendererProperties]}
 ]
 userVariables [
  User.bIsFrontEnd {type=NiagaraBool}
  User.bIsFrontEndPreview {type=NiagaraBool}
  User.ForwardDot {type=NiagaraFloat, defaultValue=0, description=""}
  User.OwnerLocalVelocity {type=Vector3f, description="", defaultValue {struct=Vector3f, value=<0 0 0>}}
  User.SkeletalMesh {
   type=NiagaraDataInterfaceSkeletalMesh
   defaultValue {
    struct=NiagaraExt_VariableValue_DataInterface
    value {dataInterface=&self:NiagaraDataInterfaceSkeletalMesh_0}
   }
  }
 ]
}
